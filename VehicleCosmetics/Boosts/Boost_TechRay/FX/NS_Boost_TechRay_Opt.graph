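# Optimized Niagara system for the TechRay vehicle boost cosmetic, placed at each boost socket on the skeletal mesh plus a per-car offset.
# ConeOptimized merges mesh and sprite layers into one emitter, using spawn groups to vary lifetime, scale and motion.
# Two more emitters add fast backward sprite streaks and per-port ribbon trails with three renderers.
# Paint variants drive colors, while User LinearColors pack material scalars fed through dynamic parameters.
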NiagaraSystem /VehicleCosmetics/Boosts/Boost_TechRay/FX/NS_Boost_TechRay_Opt
dependencies {
 usedDataInterfaces [&NiagaraDataInterfaceCurve, &NiagaraDataInterfaceArrayFloat, &NiagaraDataInterfaceSkeletalMesh]
 usedDynamicInputs []
 usedModules [
  &self:SystemSpawnScript.SystemScriptSource.SystemScriptGraph.NiagaraNodeAssignment_0.SetVariables_#1
  &SystemState
 ]
}
emitters {
 ConeOptimized {
  emitterUpdateScript [
   &EmitterState(
    Inactive Response=::NewEnumerator1
    Life Cycle Mode=::NewEnumerator1
    Loop Behavior=::NewEnumerator1
    Loop Duration Mode=::NewEnumerator1
    UseLoopDelay=false
   )
   &VC_FilteredSocketCount()
   &SpawnRate(SpawnRate=2, Use Spawn Probability=false)
   &SpawnBurst_Instantaneous(Spawn Count=4, Spawn Group=1, Use Spawn Probability=false, UseLoopCountLimit=false)
   &SpawnRate(SpawnGroup=2, SpawnRate=4, Use Spawn Probability=false) {module=SpawnRate001}
   &SpawnBurst_Instantaneous(Spawn Count=4, Spawn Group=3, Use Spawn Probability=false, UseLoopCountLimit=false) {module=SpawnBurst_Instantaneous001}
  ]
  particleSpawnScript [
   &VC_PaintedVariantsOverride(
    Black=&MultiplyLinearColor(Linear Color=<rgba 0.0078 0.0078 0.0078 1>, Scale Factor=<15 15 15 1>)
    BurntSienna=&MultiplyLinearColor(Linear Color=<rgba 0.1412 0.0627 0.0235 1>, Scale Factor=<12 12 12 1>)
    DefaultColor=&MultiplyLinearColor(Linear Color=<rgba 1 0.4 0.102 1>, Scale Factor=<6 6 6 1>)
    ForestGreen=&MultiplyLinearColor(Linear Color=<rgba 0 0.3922 0 1>, Scale Factor=<12 12 12 1>)
    Grey=&MultiplyLinearColor(Linear Color=<rgba 0.1804 0.1804 0.1804 1>, Scale Factor=<8 8 8 1>)
    Lime=&MultiplyLinearColor(Linear Color=<rgba 0.2627 0.8706 0 1>, Scale Factor=<8 8 8 1>)
    Orange=&MultiplyLinearColor(Linear Color=<rgba 1 0.2314 0 1>, Scale Factor=<4.5 4.5 4.5 1>)
    Pink=&MultiplyLinearColor(Linear Color=<rgba 0.9843 0.1804 0.7686 1>, Scale Factor=<3.5 3.5 3.5 1>)
    Purple=&MultiplyLinearColor(Linear Color=<rgba 0.4863 0 2 1>, Scale Factor=<2.25 2.25 2.25 1>)
    R1 Black:LinearColor=&MultiplyLinearColor(Linear Color=<rgba 0.0078 0.0078 0.0078 1>, Scale Factor=<15 15 15 1>)
    R1 BurntSienna:LinearColor=&MultiplyLinearColor(Linear Color=<rgba 0.1412 0.0627 0.0235 1>, Scale Factor=<12 12 12 1>)
    R1 Cobalt:LinearColor=&MultiplyLinearColor(Linear Color=<rgba 0 0.1176 1 1>, Scale Factor=<6 6 6 1>)
    R1 Crimson:LinearColor=&MultiplyLinearColor(Linear Color=<rgba 0.7412 0 0.0118 1>, Scale Factor=<10 10 10 1>)
    R1 DefaultColor=&MultiplyLinearColor(Linear Color=<rgba 1 0.4 0.102 1>, Scale Factor=<6 6 6 1>)
    R1 ForestGreen:LinearColor=&MultiplyLinearColor(Linear Color=<rgba 0 0.3922 0 1>, Scale Factor=<12 12 12 1>)
    R1 Grey:LinearColor=&MultiplyLinearColor(Linear Color=<rgba 0.1804 0.1804 0.1804 1>, Scale Factor=<8 8 8 1>)
    R1 Lime:LinearColor=&MultiplyLinearColor(Linear Color=<rgba 0.2627 0.8706 0 1>, Scale Factor=<8 8 8 1>)
    R1 Orange:LinearColor=&MultiplyLinearColor(Linear Color=<rgba 1 0.2314 0 1>, Scale Factor=<4.5 4.5 4.5 1>)
    R1 Pink:LinearColor=&MultiplyLinearColor(Linear Color=<rgba 0.9843 0.1804 0.7686 1>, Scale Factor=<3.5 3.5 3.5 1>)
    R1 Purple:LinearColor=&MultiplyLinearColor(Linear Color=<rgba 0.4863 0 2 1>, Scale Factor=<2.25 2.25 2.25 1>)
    R1 Saffron:LinearColor=&MultiplyLinearColor(Linear Color=<rgba 0.8196 0.6275 0.0078 1>, Scale Factor=<4 4 4 1>)
    R1 TitaniumWhite:LinearColor=&MultiplyLinearColor(Linear Color=<rgba 0.6 0.6 0.6 1>, Scale Factor=<2.5 2.5 2.5 1>)
    R2 Black:LinearColor=&MultiplyLinearColor(Linear Color=<rgba 0.0078 0.0078 0.0078 1>, Scale Factor=<15 15 15 1>)
    R2 BurntSienna:LinearColor=&MultiplyLinearColor(Linear Color=<rgba 0.1412 0.0627 0.0235 1>, Scale Factor=<12 12 12 1>)
    R2 Cobalt:LinearColor=&MultiplyLinearColor(Linear Color=<rgba 0 0.1176 1 1>, Scale Factor=<6 6 6 1>)
    R2 Crimson:LinearColor=&MultiplyLinearColor(Linear Color=<rgba 0.7412 0 0.0118 1>, Scale Factor=<10 10 10 1>)
    R2 DefaultColor:LinearColor=&MultiplyLinearColor(Linear Color=<rgba 1 0.4 0.102 1>, Scale Factor=<6 6 6 1>)
    R2 ForestGreen:LinearColor=&MultiplyLinearColor(Linear Color=<rgba 0 0.3922 0 1>, Scale Factor=<12 12 12 1>)
    R2 Grey:LinearColor=&MultiplyLinearColor(Linear Color=<rgba 0.1804 0.1804 0.1804 1>, Scale Factor=<8 8 8 1>)
    R2 Lime:LinearColor=&MultiplyLinearColor(Linear Color=<rgba 0.2627 0.8706 0 1>, Scale Factor=<8 8 8 1>)
    R2 Orange:LinearColor=&MultiplyLinearColor(Linear Color=<rgba 1 0.2314 0 1>, Scale Factor=<4.5 4.5 4.5 1>)
    R2 Pink:LinearColor=&MultiplyLinearColor(Linear Color=<rgba 0.9843 0.1804 0.7686 1>, Scale Factor=<3.5 3.5 3.5 1>)
    R2 Purple:LinearColor=&MultiplyLinearColor(Linear Color=<rgba 0.4863 0 2 1>, Scale Factor=<2.25 2.25 2.25 1>)
    R2 Saffron:LinearColor=&MultiplyLinearColor(Linear Color=<rgba 0.8196 0.6275 0.0078 1>, Scale Factor=<4 4 4 1>)
    R2 SkyBlue:LinearColor=&MultiplyLinearColor(Linear Color=<rgba 0 0.6 1 1>, Scale Factor=<4 4 4 1>)
    R2 TitaniumWhite:LinearColor=&MultiplyLinearColor(Linear Color=<rgba 0.6 0.6 0.6 1>, Scale Factor=<2.5 2.5 2.5 1>)
    R3 Black:LinearColor=&MultiplyLinearColor(Linear Color=<rgba 0.0078 0.0078 0.0078 1>, Scale Factor=<15 15 15 2>)
    R3 BurntSienna:LinearColor=&MultiplyLinearColor(Linear Color=<rgba 0.1412 0.0627 0.0235 1>, Scale Factor=<12 12 12 1.3>)
    R3 Cobalt:LinearColor=&MultiplyLinearColor(Linear Color=<rgba 0 0.1176 1 1>, Scale Factor=<6 6 6 1>)
    R3 Crimson:LinearColor=&MultiplyLinearColor(Linear Color=<rgba 0.7412 0 0.0118 1>, Scale Factor=<10 10 10 1>)
    R3 DefaultColor:LinearColor=&MultiplyLinearColor(Linear Color=<rgba 1 0.4 0.102 1>, Scale Factor=<6 6 6 3>)
    R3 ForestGreen:LinearColor=&MultiplyLinearColor(Linear Color=<rgba 0 0.3922 0 1>, Scale Factor=<12 12 12 1>)
    R3 Grey:LinearColor=&MultiplyLinearColor(Linear Color=<rgba 0.1804 0.1804 0.1804 1>, Scale Factor=<8 8 8 1>)
    R3 Lime:LinearColor=&MultiplyLinearColor(Linear Color=<rgba 0.2627 0.8706 0 1>, Scale Factor=<8 8 8 1>)
    R3 Orange:LinearColor=&MultiplyLinearColor(Linear Color=<rgba 1 0.2314 0 1>, Scale Factor=<4.5 4.5 4.5 1>)
    R3 Pink:LinearColor=&MultiplyLinearColor(Linear Color=<rgba 0.9843 0.1804 0.7686 1>, Scale Factor=<3.5 3.5 3.5 1>)
    R3 Purple:LinearColor=&MultiplyLinearColor(Linear Color=<rgba 0.4863 0 2 1>, Scale Factor=<2.25 2.25 2.25 1>)
    R3 Saffron:LinearColor=&MultiplyLinearColor(Linear Color=<rgba 0.8196 0.6275 0.0078 1>, Scale Factor=<4 4 4 1>)
    R3 SkyBlue:LinearColor=&MultiplyLinearColor(Linear Color=<rgba 0 0.6 1 1>, Scale Factor=<4 4 4 1>)
    R3 TitaniumWhite:LinearColor=&MultiplyLinearColor(Linear Color=<rgba 0.6 0.6 0.6 1>, Scale Factor=<2.5 2.5 2.5 2>)
    Saffron=&MultiplyLinearColor(Linear Color=<rgba 0.8196 0.6275 0.0078 1>, Scale Factor=<4 4 4 1>)
    SkyBlue=&MultiplyLinearColor(Linear Color=<rgba 0 0.6 1 1>, Scale Factor=<4 4 4 1>)
    TitaniumWhite=&MultiplyLinearColor(Linear Color=<rgba 0.6 0.6 0.6 1>, Scale Factor=<2.5 2.5 2.5 1>)
    Use R1 Painted Overrides=true
    Use R2 Painted Overrides=true
    Use R3 Painted Overrides=true
   )
   /SetVariables(Particles.SetNewFlagTag:NiagaraInt32=&SpawnGroup()) {isSetParameters=true}
   &InitializeParticle(
    Color Mode=::NewEnumerator0
    Lifetime=/VehicleCosmetics/FX/Niagara/Modules/MergeEmitters/VC_FloatBySpawnGroup(
     SpawnGroup or ExecIndex:NiagaraInt32=0
     CustomInt:NiagaraInt32=0
     MaximumSpawnGroups:NiagaraInt32=4
     Float SG 0:NiagaraFloat=&RandomRangeFloat(Minimum=0.5, Maximum=0.95)
     Float SG 1:NiagaraFloat=$Emitter.CurrentLoopDuration
     Float SG 2:NiagaraFloat=2
     Float SG 3:NiagaraFloat=$Emitter.CurrentLoopDuration
     Float SG 4:NiagaraFloat=0
     Float SG 5:NiagaraFloat=0
    )
    Mesh Scale Mode=::NewEnumerator3
    Mesh Uniform Scale=&VC_FloatBySpawnGroup(
     SpawnGroup or ExecIndex:NiagaraInt32=0
     CustomInt:NiagaraInt32=0
     MaximumSpawnGroups:NiagaraInt32=3
     Float SG 0:NiagaraFloat=0.122
     Float SG 1:NiagaraFloat=0.285
     Float SG 2:NiagaraFloat=0.122
     Float SG 3:NiagaraFloat=0
     Float SG 4:NiagaraFloat=0
     Float SG 5:NiagaraFloat=0
    )
    Sprite Rotation Mode=::NewEnumerator2
    Sprite Size Mode=::NewEnumerator3
    Uniform Sprite Size=&SelectFloatFromArray(Direct Array Index=3, Float Selection Array=DataInterface() {propertyValues {FloatData [100]}})
    UsePositionOffset=false
    Write Material Random=true
   )
   &PartitionParticles(Number Of Partitions=$Emitter.PortCount)
   &SkeletalMeshLocation(
    Bone / Socket Index=$Particles.PartitionParticles.Partition
    Bone / Socket Sampling Mode=::NewEnumerator5
    MaskSpawnGroup=false
    Mesh Sampling Type=::NewEnumerator6
    Sampled Position Offset=&AddVectorToPosition(
     Position=/VehicleCosmetics/FX/Niagara/Modules/MergeEmitters/VC_VectorBySpawnGroup(
      SpawnGroup or ExecIndex:NiagaraInt32=0
      CustomInt:NiagaraInt32=0
      MaximumSpawnGroups:NiagaraInt32=4
      Vector Group 0:Vector3f=<-82 0 0>
      Vector Group 1:Vector3f=<-83 0 0>
      Vector Group 2:Vector3f=<-75 0 0>
      Vector Group 3:Vector3f=<-86 0 0>
      Vector Group 4:Vector3f=<0 0 0>
      Vector Group 5:Vector3f=<0 0 0>
     )
     Vector=$User.DTCarOffset
    )
    Skeletal Mesh=$System.SkeletalMesh
   )
   &AddVelocity(Velocity=&VC_VectorBySpawnGroup(
     SpawnGroup or ExecIndex:NiagaraInt32=0
     CustomInt:NiagaraInt32=0
     MaximumSpawnGroups:NiagaraInt32=4
     Vector Group 0:Vector3f=<17 0 0>
     Vector Group 1:Vector3f=<0 0 0>
     Vector Group 2:Vector3f=<-12.5 0 0>
     Vector Group 3:Vector3f=<0 0 0>
     Vector Group 4:Vector3f=<0 0 0>
     Vector Group 5:Vector3f=<0 0 0>
    ))
   &Color(Color=&MakeCustomLinearColorFromBool(Bool=$User.bIsPainted, True Color=$User.PaintedColor, False Color=<rgba 1 0.4 0.1 1>)) {enabled=false}
   /SetVariables(
    Particles.CenterGlowColor:LinearColor=&MakeCustomLinearColorFromBool(Bool=$User.bIsPainted, True Color=$User.PaintedColor, False Color=<rgba 1 0.4 0.1 1>)
    Particles.ColorConeRotate:LinearColor=&MakeCustomLinearColorFromBool(Bool=$User.bIsPainted, True Color=$User.PaintedColor, False Color=<rgba 1 0.5668 0.1615 1>)
    Particles.TriTorusColor:LinearColor=&MultiplyLinearColor(
     Linear Color=&MakeCustomLinearColorFromBool(Bool=$User.bIsPainted, True Color=$User.PaintedColor, False Color=<rgba 1 0.4306 0.1458 1>)
     Scale Factor=&MakeVector4(X=1, Y=1, Z=1, W=1)
    )
   ) {isSetParameters=true}
  ]
  particleUpdateScript [
   &ParticleState()
   &UpdateMeshOrientation(
    Rotation Rate=&SelectFloatFromArray(Direct Array Index=$Particles.PartitionParticles.Partition, Float Selection Array=DataInterface() {propertyValues {FloatData [-0.3,0.3]}})
    Rotation Vector=&VC_VectorBySpawnGroup(
     SpawnGroup or ExecIndex:NiagaraInt32=0
     CustomInt:NiagaraInt32=0
     MaximumSpawnGroups:NiagaraInt32=4
     Vector Group 0:Vector3f=<0 0 0>
     Vector Group 1:Vector3f=<0 0 0>
     Vector Group 2:Vector3f=<1 0 0>
     Vector Group 3:Vector3f=<0 0 0>
     Vector Group 4:Vector3f=<0 0 0>
     Vector Group 5:Vector3f=<0 0 0>
    )
   )
   &ScaleMeshSize(Scale Factor=&VC_VectorBySpawnGroup(
     SpawnGroup or ExecIndex:NiagaraInt32=0
     CustomInt:NiagaraInt32=0
     MaximumSpawnGroups:NiagaraInt32=2
     Vector Group 0:Vector3f=&Multiply_VectorByFloat(
      Vector=<1 1 1>
      Float=&FloatFromCurve(FloatCurve=DataInterface() {propertyValues {Curve {keys [<at 0 1>,<at 1 0>]}}})
     )
     Vector Group 1:Vector3f=<1 1 1>
     Vector Group 2:Vector3f=<0 0 0>
     Vector Group 3:Vector3f=<0 0 0>
     Vector Group 4:Vector3f=<0 0 0>
     Vector Group 5:Vector3f=<0 0 0>
    ))
   &SolveForcesAndVelocity(Clamp Velocity=false, Limit Acceleration=false)
   &ScaleColor(Scale RGB=&Multiply_VectorByFloat(Vector=<1 1 1>, Float=1), ScaleA=true, ScaleRGB=true)
   &DynamicMaterialParameters(
    Index 0 Param 1=&MakeFloatFromLinearColor(
     LinearColor=Linked() {linkedVariable {name=User.LC_OpaContr/OpacityContr/OutlineBC, type=LinearColor}}
     Channel:ENiagaraLinearColor_Channels=::NewEnumerator0
    )
    Index 0 Param 2=1
    Index 0 Param 3=1
    Module.Index0Param0Write=true
    Module.Index0Param1Write=true
    Module.Index0Param2Write=true
    Module.Index0Param3Write=true
   )
   /SetVariables(
    Particles.DynamicIndex1_ConeRotate:Vector4f=&MakeVector4(
     X=&MakeFloatFromLinearColor(
      LinearColor=Linked() {linkedVariable {name=User.LC_OpaContr/OpacityContr/OutlineBC, type=LinearColor}}
      Channel:ENiagaraLinearColor_Channels=::NewEnumerator1
     )
     Y=&MakeFloatFromLinearColor(
      LinearColor=Linked() {linkedVariable {name=User.LC_OpaContr/OpacityContr/OutlineBC, type=LinearColor}}
      Channel:ENiagaraLinearColor_Channels=::NewEnumerator2
     )
     Z=1
     W=&SelectFloatFromArray(Direct Array Index=$Particles.PartitionParticles.Partition, Float Selection Array=DataInterface() {propertyValues {FloatData [1,0]}})
    )
    Particles.DynamicIndex2_ConeRotate:Vector4f=&MakeVector4(
     X=&SelectFloatFromArray(Direct Array Index=$Particles.PartitionParticles.Partition, Float Selection Array=DataInterface() {propertyValues {FloatData [0,0.5]}})
     Y=0
     Z=0
     W=0
    )
   ) {isSetParameters=true}
   /SetVariables(Particles.DynamicIndex1_TriTorus:Vector4f=&MakeVector4(
     X=&MakeFloatFromLinearColor(
      LinearColor=Linked() {linkedVariable {name=User.LC_BWControl/HuecoFres/FresIntens/AlphaBaseB, type=LinearColor}}
      Channel:ENiagaraLinearColor_Channels=::NewEnumerator0
     )
     Y=&MakeFloatFromLinearColor(
      LinearColor=Linked() {linkedVariable {name=User.LC_BWControl/HuecoFres/FresIntens/AlphaBaseB, type=LinearColor}}
      Channel:ENiagaraLinearColor_Channels=::NewEnumerator1
     )
     Z=&MakeFloatFromLinearColor(
      LinearColor=Linked() {linkedVariable {name=User.LC_BWControl/HuecoFres/FresIntens/AlphaBaseB, type=LinearColor}}
      Channel:ENiagaraLinearColor_Channels=::NewEnumerator2
     )
     W=0
    )) {isSetParameters=true}
   /SetVariables(
    Particles.ScaleColorCenterGlow:LinearColor=&MultiplyLinearColor(Linear Color=$Particles.R3 Color, Scale Factor=<1 1 1 1>)
    Particles.ScaleColorConeRotate:LinearColor=&MultiplyLinearColor(Linear Color=$Particles.R1 Color, Scale Factor=<1 1 1 1>)
    Particles.ScaleColorTriTorus:LinearColor=&MultiplyLinearColor(
     Linear Color=$Particles.R2 Color
     Scale Factor=&MakeVector4(
      X=1
      Y=1
      Z=1
      W=&FloatFromCurve(FloatCurve=DataInterface() {propertyValues {Curve {keys [<at 0 2.5>,<at 0.3 1>,<at 0.7 1>,<at 1 2.5>]}}})
     )
    )
   ) {isSetParameters=true}
  ]
  rendererClasses [&NiagaraMeshRendererProperties, &NiagaraSpriteRendererProperties]
  renderers [
   {rendererClass=&NiagaraMeshRendererProperties, rendererIndex=0}
   {rendererClass=&NiagaraMeshRendererProperties, rendererIndex=1}
   {rendererClass=&NiagaraMeshRendererProperties, rendererIndex=2}
   {rendererClass=&NiagaraSpriteRendererProperties, rendererIndex=3}
   {rendererClass=&NiagaraSpriteRendererProperties, rendererIndex=4}
  ]
 }
 NE_VC_Boost_Sprite_Parent {
  emitterUpdateScript [
   &EmitterState(Inactive Response=::NewEnumerator1, Life Cycle Mode=::NewEnumerator1, UseLoopDelay=false)
   &VC_FilteredSocketCount()
   &SpawnRate(SpawnRate=50, Use Spawn Probability=false)
  ]
  particleSpawnScript [
   &VC_PaintedVariantsOverride(
    Black=&MultiplyLinearColor(Linear Color=<rgba 0.0196 0.0196 0.0196 1>, Scale Factor=<0.1 0.1 0.1 1>)
    BurntSienna=&MultiplyLinearColor(Linear Color=<rgba 0.1412 0.0627 0.0235 1>, Scale Factor=<2 2 2 1>)
    DefaultColor=&MultiplyLinearColor(Linear Color=<rgba 1 0.4 0.1 1>, Scale Factor=<4 4 4 1>)
    ForestGreen=&MultiplyLinearColor(Linear Color=<rgba 0 0.3922 0 1>, Scale Factor=<8 8 8 1>)
    Grey=&MultiplyLinearColor(Linear Color=<rgba 0.1804 0.1804 0.1804 1>, Scale Factor=<8 8 8 1>)
    Lime=&MultiplyLinearColor(Linear Color=<rgba 0.2627 0.8706 0 1>, Scale Factor=<8 8 8 1>)
    Purple=&MultiplyLinearColor(Linear Color=<rgba 0.4863 0 2 1>, Scale Factor=<7 7 7 1>)
    Saffron=<rgba 0.8196 0.6275 0.0078 1>
    TitaniumWhite=<rgba 0.6 0.6 0.6 1>
   )
   &InitializeParticle(
    Color=$Particles.Color
    Lifetime Max=0.5
    Lifetime Min=0.35
    Lifetime Mode=::NewEnumerator1
    Sprite Rotation Mode=::NewEnumerator2
    Sprite Size=&Multiply_Vector2D_ByFloat(Float=2, Vector2D=<1 15>)
    Sprite Size Mode=::NewEnumerator4
    UsePositionOffset=false
    Write Material Random=true
   )
   &PartitionParticles(Number Of Partitions=$Emitter.PortCount)
   &SkeletalMeshLocation(
    Bone / Socket Index=$Particles.PartitionParticles.Partition
    Bone / Socket Sampling Mode=::NewEnumerator5
    MaskSpawnGroup=false
    Mesh Sampling Type=::NewEnumerator6
    Sampled Position Offset=&AddVectorToPosition(Position=<-70 0 0>, Vector=$User.DTCarOffset)
    Skeletal Mesh=$System.SkeletalMesh
   )
   &AddVelocity(Velocity=<-1000 0 0>)
   &Color(Color=&MakeCustomLinearColorFromBool(Bool=$User.bIsPainted, True Color=$User.PaintedColor, False Color=<rgba 1 0.4 0.1 1>)) {enabled=false}
   &ShapeLocation(MaskSpawnGroup=false, Offset Mode=::NewEnumerator0, Override Seed=false, Sphere Radius=20, UseOwnerScale=true)
  ]
  particleUpdateScript [
   &ParticleState()
   &SolveForcesAndVelocity(Clamp Velocity=false, Limit Acceleration=false)
   &ScaleColor(Scale RGB=&Multiply_VectorByFloat(Vector=<1 1 1>, Float=4), ScaleA=true, ScaleRGB=true)
  ]
  rendererClasses [&NiagaraSpriteRendererProperties]
  renderers [{rendererClass=&NiagaraSpriteRendererProperties}]
 }
 NE_VC_RibbonFusion {
  emitterUpdateScript [
   &EmitterState(Life Cycle Mode=::NewEnumerator1, Loop Duration=1, UseLoopDelay=false)
   &VC_FilteredSocketCount()
   &VC_RateQualityHandler(Use Custom Rate?=true) {module=VC_RateQualityHandler001}
   &VC_RateQualityHandler()
   /SetVariables(Emitter.IsInGame:NiagaraBool=&Boolean_NotOperation(Bool=$User.bIsFrontendPreview)) {isSetParameters=true}
   &VC_SpawnRibbon_v1(Number of Ribbons=$Emitter.PortCount, SpawnRate=$Emitter.Spawn Rate)
  ]
  particleSpawnScript [
   &VC_PaintedVariantsOverride() {enabled=false}
   &VC_InitializeRibbon_v1(
    Custom Offset=true
    Enable Skeletal Position=true
    Lifetime=&MakeCustomFloatFromBool(Bool=$User.bIsFrontendPreview, True Float=1, False Float=1)
    Local Vector Offset=&AddVectorToPosition(Position=<-80 0 0>, Vector=$User.DTCarOffset)
    Offset Amount=1
    Ribbon Width:NiagaraFloat=&MakeCustomFloatFromBool(Bool=$User.bIsFrontendPreview, True Float=50, False Float=50)
   )
   &AddVelocity(Velocity=&MakeCustomVectorFromBool(Bool=$User.bIsFrontendPreview, True Vector=<-1000 0 1>, False Vector=<-1000 0 0>)) {module=AddVelocity001}
   &Color(Color=&MakeCustomLinearColorFromBool(Bool=$User.bIsPainted, True Color=$User.PaintedColor, False Color=<rgba 1 1 1 1>))
   /VehicleCosmetics/FX/Niagara/Modules/MergeEmitters/VC_ExtraBindingColors(
    ExtraColors=1
    R1 Color:LinearColor=&MultiplyLinearColor(
     Linear Color=&MakeCustomLinearColorFromBool(
      Bool=$User.bIsPainted
      True Color=$User.CustomColorB
      False Color=&MakeCustomLinearColorFromBool(Bool=$User.bIsFrontendPreview, True Color=<rgba 0.1354 0.4092 1 1>, False Color=<rgba 0.0524 0.2236 0.4375 1>)
     )
     Scale Factor=&MakeVector4(
      X=1
      Y=1
      Z=1
      W=&MakeFloatFromLinearColor(
       LinearColor=Linked() {linkedVariable {name=User.LC_BWControl/HuecoFres/FresIntens/AlphaBaseB, type=LinearColor}}
       Channel:ENiagaraLinearColor_Channels=::NewEnumerator3
      )
     )
    )
   )
  ]
  particleUpdateScript [
   &ParticleState()
   &InheritSourceMovement(
    Applied Position Scale=&Multiply_VectorByFloat(
     Vector=<1 1 1>
     Float=&FloatFromCurve(FloatCurve=DataInterface() {propertyValues {Curve {keys [
          {
           arriveTangentWeight=0.4252
           interpMode=RCIM_Cubic
           leaveTangentWeight=0.4252
           tangentMode=RCTM_User
           tangentWeightMode=RCTWM_WeightedBoth
           time=0.15
           value=1
          }
          {interpMode=RCIM_Cubic, tangentMode=RCTM_User, time=1, value=0}
         ]}}})
    )
    ApplyForce=false
    ApplyPosition=true
    ApplyPositionFalloff=false
   )
   &SolveForcesAndVelocity(Clamp Velocity=false, Limit Acceleration=false)
   &VC_UpdateRibbon_v1(Enable Skeletal Position=true)
   &ScaleRibbonWidth(Ribbon Width Scale=&FloatFromCurve(FloatCurve=DataInterface() {propertyValues {Curve {keys [{arriveTangent=1.6117, interpMode=RCIM_Cubic, leaveTangent=1.6117, tangentMode=RCTM_User, time=0, value=1}]}}}))
   &DynamicMaterialParameters(
    Index 0 Param 1=&MakeCustomFloatFromBool(Bool=$User.bIsPainted, True Float=1, False Float=0)
    Index 0 Param 2=&MakeFloatFromLinearColor(
     LinearColor=Linked() {linkedVariable {name=User.LC_UseFadeMask/GlowTrailExB/GlowTrailInt/DesaturationA, type=LinearColor}}
     Channel:ENiagaraLinearColor_Channels=::NewEnumerator0
    )
    Index 0 Param 3=&MakeFloatFromLinearColor(
     LinearColor=Linked() {linkedVariable {name=User.LC_UseFadeMask/GlowTrailExB/GlowTrailInt/DesaturationA, type=LinearColor}}
     Channel:ENiagaraLinearColor_Channels=::NewEnumerator1
    )
    Index 0 Param 4=&MakeFloatFromLinearColor(
     LinearColor=Linked() {linkedVariable {name=User.LC_BaseB/BaseLinesIntensity/BaseLinesAlpha/UseSmoothFaMask, type=LinearColor}}
     Channel:ENiagaraLinearColor_Channels=::NewEnumerator1
    )
    Index 1 Param 1=&SelectFloatFromArray(Direct Array Index=$Particles.SpawnGroup, Float Selection Array=DataInterface() {propertyValues {FloatData [0.5,1]}})
    Index 1 Param 2=&SelectFloatFromArray(Direct Array Index=$Particles.SpawnGroup, Float Selection Array=DataInterface() {propertyValues {FloatData [0,1]}})
    Index 1 Param 3=&MakeFloatFromLinearColor(
     LinearColor=Linked() {linkedVariable {name=User.LC_UseFadeMask/GlowTrailExB/GlowTrailInt/DesaturationA, type=LinearColor}}
     Channel:ENiagaraLinearColor_Channels=::NewEnumerator2
    )
    Index 1 Param 4=&MakeCustomFloatFromBool(
     Bool=$User.bIsPainted
     True Float=&Multiply_Float(
      A=&MakeFloatFromLinearColor(
       LinearColor=Linked() {linkedVariable {name=User.LC_UseFadeMask/GlowTrailExB/GlowTrailInt/DesaturationA, type=LinearColor}}
       Channel:ENiagaraLinearColor_Channels=::NewEnumerator2
      )
      B=0.5
     )
     False Float=&Multiply_Float(
      A=&MakeFloatFromLinearColor(
       LinearColor=Linked() {linkedVariable {name=User.LC_UseFadeMask/GlowTrailExB/GlowTrailInt/DesaturationA, type=LinearColor}}
       Channel:ENiagaraLinearColor_Channels=::NewEnumerator2
      )
      B=1.5
     )
    )
    Index 2 Param 1=&MakeCustomFloatFromBool(
     Bool=$User.bIsPainted
     True Float=&MakeFloatFromLinearColor(
      LinearColor=Linked() {linkedVariable {name=User.LC_UseFadeMask/GlowTrailExB/GlowTrailInt/DesaturationA, type=LinearColor}}
      Channel:ENiagaraLinearColor_Channels=::NewEnumerator3
     )
     False Float=0
    )
    Index 2 Param 2=&MakeFloatFromLinearColor(
     LinearColor=Linked() {linkedVariable {name=User.LC_AlphaRayA/GlowTrailInteG/GlowTraExt/TrailRibbPow, type=LinearColor}}
     Channel:ENiagaraLinearColor_Channels=::NewEnumerator0
    )
    Index 3 Param 1=&MakeFloatFromLinearColor(
     LinearColor=Linked() {linkedVariable {name=User.LC_AlphaRayA/GlowTrailInteG/GlowTraExt/TrailRibbPow, type=LinearColor}}
     Channel:ENiagaraLinearColor_Channels=::NewEnumerator1
    )
    Module.Index0Param0Write=true
    Module.Index0Param1Write=true
    Module.Index0Param2Write=true
    Module.Index0Param3Write=true
    Module.Index1Param0Write=true
    Module.Index1Param1Write=true
    Module.Index1Param2Write=true
    Module.Index1Param3Write=true
    Module.Index2Param0Write=true
    Module.Index2Param1Write=true
    Module.Index2Param2Write=true
    Module.Index2Param3Write=true
    Module.Index3Param0Write=true
    Module.Index3Param1Write=true
    Module.Index3Param2Write=true
    Module.Index3Param3Write=true
    Write Parameter Index 1=true
    Write Parameter Index 2=true
    Write Parameter Index 3=true
   )
   &ScaleColor(
    Scale Alpha=&FloatFromCurve(
     FloatCurve=DataInterface() {propertyValues {Curve {keys [
         {arriveTangent=4.6856, interpMode=RCIM_Cubic, leaveTangent=4.6856, tangentMode=RCTM_User, time=0, value=0}
         {interpMode=RCIM_Cubic, tangentMode=RCTM_User, time=0.4, value=1}
         {arriveTangent=-3.9024, tangentMode=RCTM_User, time=0.9, value=0}
        ]}}}
     Scale Curve=&MakeCustomFloatFromBool(Bool=$User.bIsPainted, True Float=1, False Float=1)
    )
    Scale RGB=&Multiply_VectorByFloat(Vector=<1 1 1>, Float=&MakeCustomFloatFromBool(Bool=$User.bIsPainted, True Float=1, False Float=1))
    ScaleA=true
    ScaleRGB=true
   )
   /SetVariables(Particles.R1 Color:LinearColor=&VC_DynamicScaleColor(
     Color to Scale:LinearColor=&MultiplyLinearColor(
      Linear Color=&MakeCustomLinearColorFromBool(
       Bool=$User.bIsPainted
       True Color=$User.CustomColorB
       False Color=&MakeCustomLinearColorFromBool(Bool=$User.bIsFrontendPreview, True Color=<rgba 0.1354 0.4092 1 1>, False Color=<rgba 0.0524 0.2236 0.4375 1>)
      )
      Scale Factor=&MakeVector4(X=1, Y=1, Z=1, W=1.2)
     )
     RGB Scale:Vector3f=&Multiply_VectorByFloat(
      Vector=<1 1 1>
      Float=&MakeCustomFloatFromBool(
       Bool=$User.bIsPainted
       True Float=&MakeFloatFromLinearColor(
        LinearColor=Linked() {linkedVariable {name=User.LC_AlphaRayA/GlowTrailInteG/GlowTraExt/TrailRibbPow, type=LinearColor}}
        Channel:ENiagaraLinearColor_Channels=::NewEnumerator2
       )
       False Float=&MakeCustomFloatFromBool(Bool=$User.bIsFrontendPreview, True Float=0.65, False Float=0.8)
      )
     )
     Alpha Scale:NiagaraFloat=&FloatFromCurve(
      FloatCurve=DataInterface() {propertyValues {Curve {keys [
          {tangentMode=RCTM_User, time=0, value=1}
          {arriveTangent=-2.2027, interpMode=RCIM_Cubic, leaveTangent=-2.2027, tangentMode=RCTM_User, time=0.6, value=1}
          {arriveTangent=-2.8551, interpMode=RCIM_Cubic, leaveTangent=-2.8551, tangentMode=RCTM_User, time=0.99, value=0}
         ]}}}
      Scale Curve=&MakeCustomFloatFromBool(Bool=$User.bIsPainted, True Float=0.8, False Float=1)
     )
    )) {isSetParameters=true}
   /SetVariables(
    Particles.DynamicRibbonBase_Index1:Vector4f=&MakeVector4(
     X=&MakeFloatFromLinearColor(
      LinearColor=Linked() {linkedVariable {name=User.LC_AlphaRayA/GlowTrailInteG/GlowTraExt/TrailRibbPow, type=LinearColor}}
      Channel:ENiagaraLinearColor_Channels=::NewEnumerator3
     )
     Y=1.5
     Z=&MakeFloatFromLinearColor(
      LinearColor=Linked() {linkedVariable {name=User.LC_BaseB/BaseLinesIntensity/BaseLinesAlpha/UseSmoothFaMask, type=LinearColor}}
      Channel:ENiagaraLinearColor_Channels=::NewEnumerator0
     )
     W=&MakeFloatFromLinearColor(
      LinearColor=Linked() {linkedVariable {name=User.LC_BaseB/BaseLinesIntensity/BaseLinesAlpha/UseSmoothFaMask, type=LinearColor}}
      Channel:ENiagaraLinearColor_Channels=::NewEnumerator1
     )
    )
    Particles.DynamicRibbonBase_Index2:Vector4f=&MakeVector4(
     X=&MakeFloatFromLinearColor(
      LinearColor=Linked() {linkedVariable {name=User.LC_BaseB/BaseLinesIntensity/BaseLinesAlpha/UseSmoothFaMask, type=LinearColor}}
      Channel:ENiagaraLinearColor_Channels=::NewEnumerator2
     )
     Y=&MakeFloatFromLinearColor(
      LinearColor=Linked() {linkedVariable {name=User.LC_UseFadeMask/GlowTrailExB/GlowTrailInt/DesaturationA, type=LinearColor}}
      Channel:ENiagaraLinearColor_Channels=::NewEnumerator0
     )
     Z=&MakeFloatFromLinearColor(
      LinearColor=Linked() {linkedVariable {name=User.LC_BaseB/BaseLinesIntensity/BaseLinesAlpha/UseSmoothFaMask, type=LinearColor}}
      Channel:ENiagaraLinearColor_Channels=::NewEnumerator3
     )
     W=0
    )
   ) {isSetParameters=true}
  ]
  rendererClasses [&NiagaraRibbonRendererProperties]
  renderers [
   {rendererClass=&NiagaraRibbonRendererProperties}
   {rendererClass=&NiagaraRibbonRendererProperties, rendererIndex=1}
   {rendererClass=&NiagaraRibbonRendererProperties, rendererIndex=2}
  ]
 }
}
summary {
 systemName=NS_Boost_TechRay_Opt
 emitters [
  {emitterName=NE_VC_Boost_Sprite_Parent, rendererClasses [&NiagaraSpriteRendererProperties]}
  {emitterName=NE_VC_RibbonFusion, rendererClasses [&NiagaraRibbonRendererProperties]}
  {emitterName=ConeOptimized, rendererClasses [&NiagaraMeshRendererProperties, &NiagaraSpriteRendererProperties]}
 ]
 userVariables [
  User.bIsFrontendPreview {type=NiagaraBool}
  User.bIsPainted {type=NiagaraBool}
  User.ColorVariantID {type=NiagaraInt32}
  User.CustomColorB {type=LinearColor, description="", defaultValue {struct=LinearColor, value=<rgba 1 1 1 1>}}
  User.DTCarOffset {type=Vector3f, description="", defaultValue {struct=Vector3f, value=<0 0 0>}}
  User.LC_AlphaRayA/GlowTrailInteG/GlowTraExt/TrailRibbPow {type=LinearColor, description="", defaultValue {struct=LinearColor, value=<rgba 1 800 0 1>}}
  User.LC_BaseB/BaseLinesIntensity/BaseLinesAlpha/UseSmoothFaMask {type=LinearColor, description="", defaultValue {struct=LinearColor, value=<rgba 2 10 1 0>}}
  User.LC_BWControl/HuecoFres/FresIntens/AlphaBaseB {type=LinearColor, description="", defaultValue {struct=LinearColor, value=<rgba 1 40 0.6 1>}}
  User.LC_OpaContr/OpacityContr/OutlineBC {type=LinearColor, description="", defaultValue {struct=LinearColor, value=<rgba 1 5 2 0>}}
  User.LC_UseFadeMask/GlowTrailExB/GlowTrailInt/DesaturationA {type=LinearColor, description="", defaultValue {struct=LinearColor, value=<rgba 0 10 5 0.15>}}
  User.PaintedColor {type=LinearColor, description="", defaultValue {struct=LinearColor, value=<rgba 0 1 0.098 1>}}
  User.PaintedScalar {type=NiagaraFloat, defaultValue=1, description=""}
  User.PortCount {type=NiagaraInt32, defaultValue=2}
 ]
}
