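# Translucent unlit material for Niagara mesh particles, named as the black base-cone variant of the TechRay vehicle boost.
# Particle color, scaled by the ColorB&&W dynamic parameter through a Fresnel lerp whose zero exponent pins it to the scaled color, feeds MF_FX_EmissiveAdjustments.
# A separate inverted, saturated mask blends particle alpha toward black by a Fresnel term, with dynamic parameters Hueco Central (central hollow) as exponent and Intensidad as multiplier.
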Material /VehicleCosmetics/Boosts/Boost_TechRay/Materials/M_Boost_TechRay_BaseCone_CBLACK
graph {
 properties {
  bAutomaticallySetUsageInEditor=false
  bUsedWithNiagaraMeshParticles=true
  blendMode=BLEND_Translucent
  shadingModel=MSM_Unlit
  usedShadingModels=Unlit
 }
 nodes [
  MaterialExpressionMeshEmitterVertexColor_3 = ParticleColor()
  n1 = Multiply(A=@MaterialExpressionMeshEmitterVertexColor_3.A, B=Constant3Vector() {constant=<rgba 1 1 1 0>})
  n2 = Constant3Vector()
  MaterialExpressionMeshEmitterDynamicParameter_3 = DynamicParameter() {paramNames ["ColorB&&W", Hueco Central, Intensidad, Param4]}
  n3 = Multiply(
   A=Power(Base=Fresnel() {exponent=0.1}, Exp=@MaterialExpressionMeshEmitterDynamicParameter_3.Hueco Central)
   B=@MaterialExpressionMeshEmitterDynamicParameter_3.Intensidad
  )
  MaterialExpressionConstantClamp_3 = Saturate(None=OneMinus(None=LinearInterpolate(A=@n1, B=@n2, Alpha=@n3)))
  n4 = Fresnel() {exponent=0.1}
  n5 = Constant()
  MaterialFunctionCall(Emissive=LinearInterpolate(
    A=@MaterialExpressionMeshEmitterVertexColor_3.RGB
    B=Multiply(A=@MaterialExpressionMeshEmitterVertexColor_3.RGB, B=@MaterialExpressionMeshEmitterDynamicParameter_3."ColorB&&W")
    Alpha=Power(Base=@n4, Exp=@n5)
   )) {materialFunction=&MF_FX_EmissiveAdjustments}
 ]
}
