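# Translucent unlit Niagara mesh-particle material for the TechRay vehicle boost cone.
# A turbine texture scrolls in V, with dynamic parameters setting offset and blending opposite directions, while red (outline) and green channels form the pattern.
# Particle color tints it, a U-scrolling mask with adjustable power modulates it, and Fresnel blends toward particle color before MF_FX_EmissiveAdjustments.
# A saturated opacity term combines the pattern with green scaled by OpacityControl and particle alpha.
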Material /VehicleCosmetics/Boosts/Boost_TechRay/Materials/M_Boost_TechRay_BoostCone_A
graph {
 properties {
  bAutomaticallySetUsageInEditor=false
  bUsedWithNiagaraMeshParticles=true
  blendMode=BLEND_Translucent
  shadingModel=MSM_Unlit
  usedShadingModels=Unlit
 }
 nodes [
  n1 = DynamicParameter() {parameterIndex=1, defaultValue=<rgba 0 0 0 0>, paramNames [TextureOffset, Param2, Param3, Param4]}
  n2 = Add(A=TextureCoordinate(), B=AppendVector(A=Constant(), B=@n1.TextureOffset))
  MaterialExpressionMeshEmitterDynamicParameter_0 = DynamicParameter() {paramNames [OpacityControl, OpacityOutlineCont, PowerBase, RotationMirror]}
  n3 = TextureSample(UVs=LinearInterpolate(
    A=Panner(Coordinate=@n2) {speedY=-0.4}
    B=Panner(Coordinate=@n2) {speedY=0.4}
    Alpha=@MaterialExpressionMeshEmitterDynamicParameter_0.RotationMirror
   )) {samplerType=SAMPLERTYPE_LinearColor, texture=/VehicleCosmetics/Boosts/Boost_TechRay/Textures/T_VCVFX_Boost_TechRay_Turbine_RGB}
  n4 = Add(A=Multiply(A=@n3.R, B=@MaterialExpressionMeshEmitterDynamicParameter_0.OpacityOutlineCont), B=Multiply(A=@n3.G, B=Constant() {r=1}))
  MaterialExpressionMeshEmitterVertexColor_0 = ParticleColor()
  n5 = Multiply(A=@MaterialExpressionMeshEmitterDynamicParameter_0.OpacityControl, B=@MaterialExpressionMeshEmitterVertexColor_0.A)
  MaterialExpressionConstantClamp_4 = Saturate(None=Add(A=@n4, B=Multiply(A=@n3.G, B=@n5)))
  n6 = Multiply(A=@n4, B=@MaterialExpressionMeshEmitterVertexColor_0.RGB)
  n7 = Multiply(
   A=TextureCoordinate()
   B=AppendVector(A=ScalarParameter() {defaultValue=1, parameterName=TileU}, B=ScalarParameter() {defaultValue=1, parameterName=TileV})
  )
  n8 = Power(
   Base=TextureSample(UVs=Panner(Coordinate=@n7) {speedX=-2}) {samplerType=SAMPLERTYPE_LinearColor, texture=/VehicleCosmetics/Boosts/Boost_TechRay/Textures/T_VCVFX_Boost_TechRay_Mask_A}.RGB
   Exp=@MaterialExpressionMeshEmitterDynamicParameter_0.PowerBase
  )
  n9 = Constant() {r=1}
  n10 = Power(Base=Fresnel(), Exp=Constant() {r=1.4})
  n11 = Constant() {r=2}
  MaterialFunctionCall(Emissive=LinearInterpolate(A=Multiply(A=@n6, B=@n8), B=Multiply(A=@n9, B=@MaterialExpressionMeshEmitterVertexColor_0.RGB), Alpha=Multiply(A=@n10, B=@n11))) {materialFunction=&MF_FX_EmissiveAdjustments}
 ]
}
