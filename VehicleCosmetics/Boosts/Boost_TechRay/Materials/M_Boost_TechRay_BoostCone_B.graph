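# Translucent unlit material for the TechRay vehicle boost cone, used on Niagara mesh particles.
# A scrolling BaseCone texture is tinted by particle color and blended toward flat particle color by a scaled Fresnel term, then fed into MF_FX_EmissiveAdjustments.
# A saturated mask multiplies the cone texture by particle alpha, the OPAControl dynamic parameter and a second texture panning along the other axis.
# Both textures tile through TileU and TileV parameters.
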Material /VehicleCosmetics/Boosts/Boost_TechRay/Materials/M_Boost_TechRay_BoostCone_B
graph {
 properties {
  bAutomaticallySetUsageInEditor=false
  bUsedWithNiagaraMeshParticles=true
  blendMode=BLEND_Translucent
  shadingModel=MSM_Unlit
  usedShadingModels=Unlit
 }
 nodes [
  n1 = TextureCoordinate()
  n2 = AppendVector(A=ScalarParameter() {defaultValue=1, parameterName=TileU}, B=ScalarParameter() {defaultValue=10, parameterName=TileV})
  n3 = TextureSample(UVs=Panner(Coordinate=Multiply(A=@n1, B=@n2)) {speedY=-0.4}) {samplerType=SAMPLERTYPE_LinearColor, texture=/VehicleCosmetics/Boosts/Boost_TechRay/Textures/T_VCVFX_Boost_TechRay_BaseCone_A}
  MaterialExpressionMeshEmitterVertexColor_2 = ParticleColor()
  MaterialExpressionMeshEmitterDynamicParameter_2 = DynamicParameter() {paramNames [OPAControl,Param2,Param3,Param4]}
  n4 = Multiply(A=@MaterialExpressionMeshEmitterVertexColor_2.A, B=@MaterialExpressionMeshEmitterDynamicParameter_2.OPAControl)
  n5 = ScalarParameter() {defaultValue=1, parameterName=TileU}
  n6 = ScalarParameter() {defaultValue=10, parameterName=TileV}
  n7 = Panner(Coordinate=Multiply(A=TextureCoordinate(), B=AppendVector(A=@n5, B=@n6))) {speedX=-2}
  MaterialExpressionConstantClamp_2 = Saturate(None=Multiply(
    A=Multiply(A=@n3.RGB, B=@n4)
    B=TextureSample(UVs=@n7) {samplerType=SAMPLERTYPE_LinearColor, texture=/VehicleCosmetics/Boosts/Boost_TechRay/Textures/T_VCVFX_Boost_TechRay_Mask_A}.RGB
   ))
  n8 = Constant() {r=1}
  n9 = Power(Base=Fresnel(), Exp=Constant() {r=0.5})
  n10 = Constant() {r=3}
  MaterialFunctionCall(Emissive=LinearInterpolate(
    A=Multiply(A=@n3.RGB, B=@MaterialExpressionMeshEmitterVertexColor_2.RGB)
    B=Multiply(A=@n8, B=@MaterialExpressionMeshEmitterVertexColor_2.RGB)
    Alpha=Multiply(A=@n9, B=@n10)
   )) {materialFunction=&MF_FX_EmissiveAdjustments}
 ]
}
