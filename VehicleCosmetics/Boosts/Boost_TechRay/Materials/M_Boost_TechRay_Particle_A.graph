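# Unlit translucent Niagara sprite material for the TechRay vehicle boost effect, built on the boost cone inner flare texture.
# The flare RGB is tinted by particle color and passed through MF_FX_EmissiveAdjustments as emissive.
# A saturated opacity-style mask multiplies the same RGB by particle alpha and the OpaI dynamic parameter, allowing per-particle intensity control.
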Material /VehicleCosmetics/Boosts/Boost_TechRay/Materials/M_Boost_TechRay_Particle_A
graph {
 properties {
  bAutomaticallySetUsageInEditor=false
  bUsedWithNiagaraSprites=true
  blendMode=BLEND_Translucent
  shadingModel=MSM_Unlit
  usedShadingModels=Unlit
 }
 nodes [
  n1 = TextureSample() {samplerType=SAMPLERTYPE_LinearColor, texture=/VehicleCosmetics/Boosts/Boost_TechRay/Textures/T_VCVFX_Boost_TechRay_BoostConeInnerFlare_A}
  MaterialExpressionVertexColor_0 = ParticleColor()
  n2 = DynamicParameter() {paramNames [OpaI,Param2,Param3,Param4]}
  MaterialExpressionConstantClamp_4 = Saturate(None=Multiply(A=@n1.RGB, B=Multiply(A=@MaterialExpressionVertexColor_0.A, B=@n2.OpaI)))
  MaterialFunctionCall(Emissive=Multiply(A=@n1.RGB, B=@MaterialExpressionVertexColor_0.RGB)) {materialFunction=&MF_FX_EmissiveAdjustments}
 ]
}
