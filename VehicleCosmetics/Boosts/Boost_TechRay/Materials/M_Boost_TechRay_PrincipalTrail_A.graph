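# Unlit translucent Niagara ribbon material for the principal trail of the TechRay vehicle boost cosmetic.
# Panning line texture, static base texture and faster-panning mask are tinted by particle color and summed into MF_FX_EmissiveAdjustments.
# Alpha combines the lines with an inverted-U gradient, then multiplies by particle alpha and a sphere-mask or gradient fade.
# Two dynamic parameter sets drive intensities and select the fade mode through If nodes.
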Material /VehicleCosmetics/Boosts/Boost_TechRay/Materials/M_Boost_TechRay_PrincipalTrail_A
graph {
 properties {
  bAutomaticallySetUsageInEditor=false
  bUsedWithNiagaraRibbons=true
  blendMode=BLEND_Translucent
  shadingModel=MSM_Unlit
  usedShadingModels=Unlit
 }
 nodes [
  n1 = DynamicParameter() {parameterIndex=1, defaultValue=<rgba 1 0 0 1>, paramNames [BaseLinesAlphaIntensity, UseFadeMask, UseSmoothFadeMask, Param4]}
  n2 = TextureCoordinate()
  n3 = AppendVector(A=ScalarParameter() {defaultValue=5, parameterName=TileU3}, B=ScalarParameter() {defaultValue=1, parameterName=TileV3})
  n4 = TextureSample(UVs=Panner(Coordinate=Multiply(A=@n2, B=@n3)) {speedX=-0.5}) {samplerType=SAMPLERTYPE_LinearColor, texture=/VehicleCosmetics/Boosts/Boost_TechRay/Textures/T_VCVFX_Boost_TechRay_InnerD2_A}
  n5 = OneMinus(None=ComponentMask(None=TextureCoordinate()) {r=true})
  n6 = DynamicParameter() {defaultValue=<rgba 1 1 5 7>, paramNames [RibbonBasePow, PanningPow, BaseB, BaseLinesIntensity]}
  n7 = Multiply(A=@n5, B=@n6.RibbonBasePow)
  n8 = Add(A=Multiply(A=Multiply(A=@n1.BaseLinesAlphaIntensity, B=@n4.RGB), B=@n5), B=@n7)
  MaterialExpressionVertexColor_0 = ParticleColor()
  n9 = TextureCoordinate() {vTiling=0}
  n10 = Constant2Vector() {r=0.6}
  n11 = Saturate(None=Multiply(A=SphereMask(A=@n9, B=@n10) {attenuationRadius=0.55, hardnessPercent=10}) {constB=8})
  n12 = Multiply(A=OneMinus(None=@n7)) {constB=8}
  n13 = Multiply(A=@n12, B=SphereMask(A=TextureCoordinate() {vTiling=0}, B=Constant2Vector() {r=1}) {attenuationRadius=1, hardnessPercent=10})
  n14 = Saturate(None=If(A=@n1.UseSmoothFadeMask, "A > B"=@n13, "A == B"=@n13, "A < B"=@n12) {constB=1})
  MaterialExpressionConstantClamp_0 = Saturate(None=Multiply(A=Multiply(A=@n8, B=@MaterialExpressionVertexColor_0.A), B=If(A=@n1.UseFadeMask, "A > B"=@n11, "A == B"=@n11, "A < B"=@n14) {constB=1}))
  Constant() {r=7}
  n15 = Constant()
  n16 = Multiply(A=@n15, B=@n15)
  n17 = Multiply(
   A=@n6.BaseB
   B=StaticSwitchParameter(True=Multiply(A=@n16, B=@MaterialExpressionVertexColor_0.RGB), False=Constant3Vector() {constant=<rgba 1 1 1 0>}) {parameterName=Color}
  )
  n18 = TextureSample() {samplerType=SAMPLERTYPE_LinearColor, texture=/VehicleCosmetics/Boosts/Boost_TechRay/Textures/T_VCVFX_Boost_TechRay_InnerG_A}
  n19 = Add(
   A=Multiply(A=Multiply(A=@n6.BaseLinesIntensity, B=@n4.RGB), B=@MaterialExpressionVertexColor_0.RGB)
   B=Add(B=Multiply(A=@n17, B=@n18.RGB)) {constA=0.15}
  )
  n20 = Multiply(
   A=TextureCoordinate()
   B=AppendVector(A=ScalarParameter() {defaultValue=1, parameterName=TileU}, B=ScalarParameter() {defaultValue=1, parameterName=TileV})
  )
  n21 = Multiply(
   A=@n6.PanningPow
   B=TextureSample(UVs=Panner(Coordinate=@n20) {speedX=-2}) {samplerType=SAMPLERTYPE_LinearColor, texture=/VehicleCosmetics/Boosts/Boost_TechRay/Textures/T_VCVFX_Boost_TechRay_MaskOne_A}.RGB
  )
  MaterialFunctionCall(Emissive=Add(A=Multiply(A=@n19, B=@MaterialExpressionVertexColor_0.RGB), B=Multiply(A=@MaterialExpressionVertexColor_0.RGB, B=@n21))) {materialFunction=&MF_FX_EmissiveAdjustments}
 ]
}
