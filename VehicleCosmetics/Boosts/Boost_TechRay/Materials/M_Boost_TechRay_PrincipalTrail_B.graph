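# Translucent unlit material for Niagara ribbons, drawing the principal trail of the TechRay vehicle boost.
# Three tiled textures pan along U at different speeds, two sharing UVs shifted by a dynamic OffsetTrail parameter.
# The InnerA texture masks the sum of InnerB and InnerF, which is scaled by particle alpha and saturated.
# Particle color tints InnerA before it passes through MF_FX_EmissiveAdjustments.
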Material /VehicleCosmetics/Boosts/Boost_TechRay/Materials/M_Boost_TechRay_PrincipalTrail_B
graph {
 properties {
  bAutomaticallySetUsageInEditor=false
  bUsedWithNiagaraRibbons=true
  blendMode=BLEND_Translucent
  shadingModel=MSM_Unlit
  usedShadingModels=Unlit
 }
 nodes [
  n1 = Constant()
  AppendVector(A=@n1, B=@n1)
  n2 = DynamicParameter() {paramNames [OffsetTrail,Param2,Param3,Param4]}
  n3 = Constant()
  n4 = Multiply(
   A=Add(A=AppendVector(A=@n2.OffsetTrail, B=@n3), B=TextureCoordinate())
   B=AppendVector(A=ScalarParameter() {defaultValue=8, parameterName=TileU}, B=ScalarParameter() {defaultValue=1, parameterName=TileV})
  )
  n5 = TextureSample(UVs=Panner(Coordinate=@n4) {speedX=-0.4}) {samplerType=SAMPLERTYPE_LinearColor, texture=/VehicleCosmetics/Boosts/Boost_TechRay/Textures/T_VCVFX_Boost_TechRay_InnerA_A}
  n6 = Multiply(
   A=TextureCoordinate()
   B=AppendVector(A=ScalarParameter() {defaultValue=16, parameterName=TileU2}, B=ScalarParameter() {defaultValue=0.9, parameterName=TileV2})
  )
  n7 = Add(
   A=TextureSample(UVs=Panner(Coordinate=@n4) {speedX=-1}) {samplerType=SAMPLERTYPE_LinearColor, texture=/VehicleCosmetics/Boosts/Boost_TechRay/Textures/T_VCVFX_Boost_TechRay_InnerB_A}.RGB
   B=TextureSample(UVs=Panner(Coordinate=@n6) {speedX=-0.5}) {samplerType=SAMPLERTYPE_LinearColor, texture=/VehicleCosmetics/Boosts/Boost_TechRay/Textures/T_VCVFX_Boost_TechRay_InnerF_A}.RGB
  )
  MaterialExpressionVertexColor_2 = ParticleColor()
  MaterialExpressionConstantClamp_3 = Saturate(None=Multiply(A=Multiply(A=@n5.RGB, B=@n7), B=@MaterialExpressionVertexColor_2.A))
  MaterialFunctionCall(Emissive=Multiply(A=@MaterialExpressionVertexColor_2.RGB, B=@n5.RGB)) {materialFunction=&MF_FX_EmissiveAdjustments}
 ]
}
