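# Unlit translucent Niagara ribbon material for the principal trail of the TechRay vehicle boost cosmetic.
# Dynamic parameters offset U and mirror V when Flip is below one, then TileU/TileV scale the UVs and a Panner scrolls them along U.
# The texture's red channel multiplies particle color into the emissive adjustments function, and particle alpha into a saturated mask.
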Material /VehicleCosmetics/Boosts/Boost_TechRay/Materials/M_Boost_TechRay_PrincipalTrail_C
graph {
 properties {
  bAutomaticallySetUsageInEditor=false
  bUsedWithNiagaraRibbons=true
  blendMode=BLEND_Translucent
  shadingModel=MSM_Unlit
  usedShadingModels=Unlit
 }
 nodes [
  MaterialExpressionVertexColor_0 = ParticleColor()
  n1 = DynamicParameter() {paramNames [OffsetTrail,Flip,Param3,Param4]}
  n2 = TextureCoordinate()
  n3 = Add(
   A=AppendVector(A=@n1.OffsetTrail, B=Constant())
   B=If(A=@n1.Flip, B=Constant() {r=1}, "A > B"=@n2, "A == B"=@n2, "A < B"=TextureCoordinate() {vTiling=-1})
  )
  n4 = AppendVector(A=ScalarParameter() {defaultValue=8, parameterName=TileU}, B=ScalarParameter() {defaultValue=1, parameterName=TileV})
  n5 = TextureSample(UVs=Panner(Coordinate=Multiply(A=@n3, B=@n4)) {speedX=-2}) {samplerType=SAMPLERTYPE_LinearColor, texture=/VehicleCosmetics/Boosts/Boost_TechRay/Textures/T_VCVFX_Boost_TechRay_InnerB2_RGB}
  MaterialExpressionConstantClamp_0 = Saturate(None=Multiply(A=@MaterialExpressionVertexColor_0.A, B=@n5.R))
  MaterialFunctionCall(Emissive=Multiply(A=@MaterialExpressionVertexColor_0.RGB, B=@n5.R)) {materialFunction=&MF_FX_EmissiveAdjustments}
 ]
}
