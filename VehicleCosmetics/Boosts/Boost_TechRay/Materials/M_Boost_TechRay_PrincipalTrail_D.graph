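# Translucent unlit material for Niagara ribbons, named as the principal trail of the TechRay vehicle boost effect.
# It samples the TechRay InnerH texture and tints its RGB by particle color, feeding that into the shared MF_FX_EmissiveAdjustments function.
# A second multiply scales the same texture RGB by particle alpha, so Niagara drives both tint and fade.
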Material /VehicleCosmetics/Boosts/Boost_TechRay/Materials/M_Boost_TechRay_PrincipalTrail_D
graph {
 properties {
  bAutomaticallySetUsageInEditor=false
  bUsedWithNiagaraRibbons=true
  blendMode=BLEND_Translucent
  shadingModel=MSM_Unlit
  usedShadingModels=Unlit
 }
 nodes [
  n1 = TextureSample() {samplerType=SAMPLERTYPE_LinearColor, texture=/VehicleCosmetics/Boosts/Boost_TechRay/Textures/T_VCVFX_Boost_TechRay_InnerH_A}
  n2 = ParticleColor()
  MaterialFunctionCall(Emissive=Multiply(A=@n1.RGB, B=@n2.RGB)) {materialFunction=&MF_FX_EmissiveAdjustments}
  Multiply(A=@n1.RGB, B=@n2.A)
 ]
}
