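# Material function that computes Border Color and Border Opacity for the TechRay vehicle boost effect.
# An inner texture, amplified, saturated and scaled by BorderIntensity, multiplies a color blended from default to particle color by IsPainted.
# Opacity combines that mask with particle alpha, scaled by default opacity when unpainted, and a soft U-axis SphereMask fade.
# A trailing If that selects the color is left unconnected.
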MaterialFunction /VehicleCosmetics/Boosts/Boost_TechRay/Materials/MF_Boost_TechRay_Border
graph {nodes [
  n1 = TextureSample() {samplerType=SAMPLERTYPE_LinearColor, texture=/VehicleCosmetics/Boosts/Boost_TechRay/Textures/T_VCVFX_Boost_TechRay_InnerH_A}
  n2 = Multiply(A=Saturate(None=Multiply(A=@n1.RGB) {constB=7}), B=FunctionInput() {inputName=BorderIntensity, inputType=FunctionInput_Scalar})
  n3 = FunctionInput() {inputName=Border DefaultColor}
  n4 = ParticleColor()
  n5 = FunctionInput() {inputName=IsPainted, inputType=FunctionInput_Scalar}
  FunctionOutput(None=Multiply(A=@n2, B=LinearInterpolate(A=@n3, B=@n4.RGB, Alpha=@n5))) {outputName=Border Color}
  n6 = If(
   A=@n5
   "A > B"=@n4.A
   "A == B"=@n4.A
   "A < B"=Multiply(A=@n4.A, B=FunctionInput() {inputName=Border DefaultOpacity, inputType=FunctionInput_Scalar, sortPriority=1})
  ) {constB=1}
  n7 = SphereMask(A=TextureCoordinate() {vTiling=0}, B=Constant2Vector()) {attenuationRadius=1, hardnessPercent=0}
  MaterialExpressionConstantClamp_0 = Saturate(None=Multiply(A=Multiply(A=@n2, B=@n6), B=Multiply(A=@n7) {constB=4}))
  FunctionOutput(None=@MaterialExpressionConstantClamp_0) {outputName=Border Opacity, sortPriority=1}
  If(A=@n5, "A > B"=@n4.RGB, "A == B"=@n4.RGB, "A < B"=@n3) {constB=1}
 ]}
