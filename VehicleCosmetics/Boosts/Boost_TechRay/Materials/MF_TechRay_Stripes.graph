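# Material function for the TechRay vehicle boost that outputs color and opacity for scrolling interior trail stripes.
# IsPainted selects particle color and alpha when at least 1, otherwise the default color and particle alpha scaled by default opacity.
# Color tints the panning InnerA texture, scaled by dynamic parameter GlowTrailInteriorG.
# Opacity multiplies InnerA by two summed, differently panned textures; OffsetTrail shifts some UVs, and Stripes Intensity is unconnected.
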MaterialFunction /VehicleCosmetics/Boosts/Boost_TechRay/Materials/MF_TechRay_Stripes
graph {nodes [
  n1 = FunctionInput() {inputName=IsPainted, inputType=FunctionInput_Scalar}
  n2 = ParticleColor()
  n3 = If(A=@n1, "A > B"=@n2.RGB, "A == B"=@n2.RGB, "A < B"=FunctionInput() {inputName=Stripes DefaultColor, previewValue=<1 0.4 0.1 1>}) {constB=1}
  n4 = ScalarParameter() {parameterName=OffsetTrail}
  n5 = Add(A=AppendVector(A=@n4, B=Constant() {r=-0.075}), B=TextureCoordinate())
  n6 = AppendVector(A=ScalarParameter() {defaultValue=8, parameterName=TileU}, B=ScalarParameter() {defaultValue=1.15, parameterName=TileV})
  n7 = TextureSample(UVs=Panner(Coordinate=Multiply(A=@n5, B=@n6)) {speedX=-0.4}) {samplerType=SAMPLERTYPE_LinearColor, texture=/VehicleCosmetics/Boosts/Boost_TechRay/Textures/T_VCVFX_Boost_TechRay_InnerA_A}
  FunctionOutput(None=Multiply(
    A=Multiply(A=@n3, B=@n7.RGB)
    B=DynamicParameter() {parameterIndex=3, paramNames [GlowTrailInteriorG, Param2, Param3, Param4]}.GlowTrailInteriorG
   )) {outputName=Stripes Color}
  n8 = Constant() {r=-0.075}
  n9 = Multiply(
   A=Add(A=AppendVector(A=@n4, B=@n8), B=TextureCoordinate())
   B=AppendVector(A=ScalarParameter() {defaultValue=8, parameterName=TileU}, B=ScalarParameter() {defaultValue=1.15, parameterName=TileV})
  )
  n10 = Multiply(
   A=TextureCoordinate()
   B=AppendVector(A=ScalarParameter() {defaultValue=16, parameterName=TileU2}, B=ScalarParameter() {defaultValue=0.9, parameterName=TileV2})
  )
  n11 = Add(
   A=TextureSample(UVs=Panner(Coordinate=@n9) {speedX=-1}) {samplerType=SAMPLERTYPE_LinearColor, texture=/VehicleCosmetics/Boosts/Boost_TechRay/Textures/T_VCVFX_Boost_TechRay_InnerB_A}.RGB
   B=TextureSample(UVs=Panner(Coordinate=@n10) {speedX=-0.5}) {samplerType=SAMPLERTYPE_LinearColor, texture=/VehicleCosmetics/Boosts/Boost_TechRay/Textures/T_VCVFX_Boost_TechRay_InnerF_A}.RGB
  )
  n12 = FunctionInput() {inputName=Stripes DefaultOpacity, inputType=FunctionInput_Scalar, sortPriority=1, previewValue=<1 0 0 1>}
  n13 = Multiply(A=@n2.A, B=@n12)
  MaterialExpressionConstantClamp_3 = Saturate(None=Multiply(A=Multiply(A=@n7.RGB, B=@n11), B=If(A=@n1, "A > B"=@n2.A, "A == B"=@n2.A, "A < B"=@n13) {constB=1}))
  FunctionOutput(None=@MaterialExpressionConstantClamp_3) {outputName=Stripes Opacity, sortPriority=1}
  ScalarParameter() {defaultValue=10, parameterName=Stripes Intensity}
 ]}
