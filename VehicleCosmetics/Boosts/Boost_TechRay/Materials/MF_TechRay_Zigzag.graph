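# Material function for the zigzag layer of the TechRay vehicle boost VFX, outputting its color and opacity.
# Dynamic parameters offset and optionally V-flip the tiled, panned UVs of one texture, whose R and G channels become separately scaled outer and inner glows before desaturation.
# IsPainted switches color and opacity between particle color and supplied defaults.
# Opacity uses only R, scaled by AlphaRayoA and faded along U with a soft SphereMask.
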MaterialFunction /VehicleCosmetics/Boosts/Boost_TechRay/Materials/MF_TechRay_Zigzag
graph {nodes [
  n1 = DynamicParameter() {parameterIndex=1, paramNames [OffsetTrail, Flip, InnerIntensity, OuterIntensity]}
  n2 = Constant() {r=1}
  n3 = Constant() {r=-0.075}
  n4 = Constant() {r=0.075}
  n5 = TextureCoordinate()
  n6 = Add(
   A=AppendVector(A=@n1.OffsetTrail, B=If(A=@n1.Flip, B=@n2, "A > B"=@n3, "A == B"=@n3, "A < B"=@n4))
   B=If(A=@n1.Flip, B=Constant() {r=1}, "A > B"=@n5, "A == B"=@n5, "A < B"=TextureCoordinate() {vTiling=-1})
  )
  n7 = AppendVector(A=ScalarParameter() {defaultValue=3, parameterName=TileU}, B=ScalarParameter() {defaultValue=1.15, parameterName=TileV})
  n8 = TextureSample(UVs=Panner(Coordinate=Multiply(A=@n6, B=@n7)) {speedX=-2}) {samplerType=SAMPLERTYPE_LinearColor, texture=/VehicleCosmetics/Boosts/Boost_TechRay/Textures/T_VCVFX_Boost_TechRay_InnerB2_RGB}
  n9 = FunctionInput() {inputName=IsPainted, inputType=FunctionInput_Scalar}
  n10 = ParticleColor()
  n11 = If(A=@n9, "A > B"=@n10.RGB, "A == B"=@n10.RGB, "A < B"=FunctionInput() {inputName=Zigzag DefaultColor}) {constB=1}
  n12 = Multiply(A=Multiply(A=@n8.R, B=@n11), B=@n1.OuterIntensity) {constB=6}
  n13 = Multiply(A=Multiply(A=@n11, B=@n8.G), B=@n1.InnerIntensity) {constB=2}
  n14 = DynamicParameter() {parameterIndex=2, defaultValue=<rgba 0 1 1 1>, paramNames [Desaturation,AlphaRayoA,Param3,Param4]}
  FunctionOutput(None=Desaturation(None=Add(A=@n12, B=@n13), Fraction=@n14.Desaturation)) {outputName=Zigzag Color}
  n15 = Multiply(
   A=@n14.AlphaRayoA
   B=If(
    A=@n9
    "A > B"=@n10.A
    "A == B"=@n10.A
    "A < B"=FunctionInput() {inputName="Zigzag  DefaultOpacity", inputType=FunctionInput_Scalar, sortPriority=1}
   ) {constB=1}
  )
  n16 = Add(A=@n8.R, B=@n8.R)
  n17 = SphereMask(A=TextureCoordinate() {vTiling=0}, B=Constant2Vector()) {attenuationRadius=1, hardnessPercent=0}
  MaterialExpressionConstantClamp_0 = Saturate(None=Multiply(A=Multiply(A=@n15, B=@n16), B=Multiply(A=@n17) {constB=4}))
  FunctionOutput(None=@MaterialExpressionConstantClamp_0) {outputName=Zigzag Opacity, sortPriority=1}
 ]}
