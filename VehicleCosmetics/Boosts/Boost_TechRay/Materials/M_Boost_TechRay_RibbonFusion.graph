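# Translucent unlit Niagara ribbon material for the TechRay vehicle boost cosmetic, layering border, zigzag and stripe material functions tinted from one yellow parameter.
# Emissive sums the layers, cutting stripe opacity from zigzag and squaring stripe color, then passes through shared FX adjustments.
# Summed opacities are shaped by U-only sphere masks, switched by the particle's UseFadeMask.
# Dynamic parameters also drive painted state, though the border's IsPainted input receives BorderIntensity.
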Material /VehicleCosmetics/Boosts/Boost_TechRay/Materials/M_Boost_TechRay_RibbonFusion
graph {
 properties {
  bAutomaticallySetUsageInEditor=false
  bUsedWithNiagaraRibbons=true
  blendMode=BLEND_Translucent
  shadingModel=MSM_Unlit
  usedShadingModels=Unlit
 }
 nodes [
  n1 = VectorParameter() {parameterName=Yellow Color, defaultValue=<rgba 1 0.4 0.102 1>}
  n2 = DynamicParameter() {defaultValue=<rgba 0 0 1 1>, paramNames [IsPainted, UseFadeMask, BorderIntensity, Param4]}
  n3 = MaterialFunctionCall!(Border DefaultColor=@n1.RGB, BorderIntensity=@n2.BorderIntensity, IsPainted=@n2.BorderIntensity, Border DefaultOpacity=@n1.RGB) {materialFunction=&/MF_Boost_TechRay_Border}
  n4 = MaterialFunctionCall!(Zigzag DefaultColor=@n1.RGB, IsPainted=@n2.IsPainted, "Zigzag  DefaultOpacity"=@n1.RGB) {materialFunction=&/MF_TechRay_Zigzag}
  n5 = MaterialFunctionCall!(Stripes DefaultColor=@n1.RGB, IsPainted=@n2.IsPainted, Stripes DefaultOpacity=@n1.RGB) {materialFunction=&/MF_TechRay_Stripes}
  n6 = Multiply(A=@n4.Zigzag Color, B=Subtract(A=@n4.Zigzag Opacity, B=@n5.Stripes Opacity))
  n7 = Multiply(A=@n5.Stripes Color, B=@n5.Stripes Color)
  MaterialFunctionCall(Emissive=Add(A=@n3.Border Color, B=Add(A=@n6, B=@n7))) {materialFunction=&MF_FX_EmissiveAdjustments}
  n8 = SphereMask(A=TextureCoordinate() {vTiling=0}, B=Constant2Vector() {r=0.6}) {attenuationRadius=0.53, hardnessPercent=10}
  n9 = Multiply(A=@n3.Border Opacity, B=Saturate(None=Multiply(A=@n8) {constB=5}))
  n10 = If(A=@n2.UseFadeMask, "A > B"=@n9, "A == B"=@n9, "A < B"=@n3.Border Opacity) {constB=1}
  n11 = TextureCoordinate() {vTiling=0}
  n12 = Constant2Vector() {r=0.6}
  n13 = Saturate(None=Multiply(A=SphereMask(A=@n11, B=@n12) {attenuationRadius=0.59, hardnessPercent=0}) {constB=8})
  n14 = Multiply(A=SphereMask(A=TextureCoordinate() {vTiling=0}, B=Constant2Vector() {r=0.6}) {attenuationRadius=0.58, hardnessPercent=0}) {constB=2}
  Multiply(
   A=Add(A=Add(A=@n10, B=@n4.Zigzag Opacity), B=@n5.Stripes Opacity)
   B=If(A=@n2.UseFadeMask, "A > B"=@n13, "A == B"=@n13, "A < B"=Saturate(None=@n14)) {constB=1}
  )
 ]
}
